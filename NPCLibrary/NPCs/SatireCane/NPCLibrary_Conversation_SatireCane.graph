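# Conversation database for the SatireCane NPC, whose greetings are Power Rangers themed.
# The entry point feeds one speech node with a default line and contextual lines gated by controller requirements.
# A first-ever-meeting line takes its priority from a curve row, while three lines require a match conversation count above a threshold.
# The speech then branches through link nodes to the shared NPCLibrary quests entry and a SatireCane sale entry, both defined elsewhere.
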ConversationDatabase /NPCLibrary/NPCs/SatireCane/NPCLibrary_Conversation_SatireCane
entry_points [{tag=Conversation.Entry.SatireCane, destinations [#1]}]
nodes [
 ConversationEntryPointNode_7 {
  id=#1
  class=ConversationEntryPointNode
  outputs [#2]
  properties {entryTag {tagName=Conversation.Entry.SatireCane}}
 }
 Speech_C_4 {
  id=#2
  class=Speech_C
  outputs [#3, #4]
  properties {
   nodeName=""
   generalConfig {
    defaultMessage=Is it morphin' time already?
    contextualMessages [
     {
      message=My friends and I heard you needed some help fighting evil.
      contextRequirements [{requirement {class=FortControllerRequirement_IsFirstEverConversationWithNPC}}]
      priority {curve {rowName=Tandem.Speech.Priority.FirstEverInteraction}}
     }
     {
      message=May the power protect you.
      contextRequirements [{requirement {
         class=FortControllerRequirement_CheckNPCMatchConversationCount
         properties {nPCUniqueIdTag {tagName=AISpawnerData.Type.Tandem.SatireCane}, requiredConversationCountConfig {requirement=GreaterThan}}
        }}]
      priority {curve {curveTable=null, rowName=None}}
     }
     {
      message=The world needs the Power Rangers.
      contextRequirements [{requirement {
         class=FortControllerRequirement_CheckNPCMatchConversationCount
         properties {nPCUniqueIdTag {tagName=AISpawnerData.Type.Tandem.SatireCane}, requiredConversationCountConfig {requirement=GreaterThan}}
        }}]
      priority {curve {curveTable=null, rowName=None}}
     }
     {
      message=Is it morphin' time already?
      contextRequirements [{requirement {
         class=FortControllerRequirement_CheckNPCMatchConversationCount
         properties {nPCUniqueIdTag {tagName=AISpawnerData.Type.Tandem.SatireCane}, requiredConversationCountConfig {requirement=GreaterThan}}
        }}]
      priority {curve {curveTable=null, rowName=None}}
     }
    ]
   }
  }
 }
 ConversationLinkNode_2 {id=#4, class=ConversationLinkNode, properties {remoteEntryTag {tagName=Conversation.Entry.SatireCane.Sale}}}
 ConversationLinkNode_1 {id=#3, class=ConversationLinkNode, properties {remoteEntryTag {tagName=Conversation.Entry.NPCLibrary.Quests}}}
]
